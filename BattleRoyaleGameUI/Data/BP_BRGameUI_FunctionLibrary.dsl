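; Blueprint function library for the Battle Royale game UI that supplies visuals (message, colors, icon, alert flags) for each game phase step.
; GetGamePhaseVisualsFromStep looks up the step in a local map variable, logging and returning an empty-message struct when none is found.
; GetOverrideColor replaces primary and secondary colors with named vector parameters from the MPC_BRGameUI material parameter collection, keeping the original color when no name is set.
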(fn GetGamePhaseVisualsFromStep
  (params (GamePhaseStep :type EAthenaGamePhaseStep%20Enum))
  (bind (_messagetext_4_ca9786984f68487451e3e6bf33553019 _primarycolorname_28_e8470e2a467ff6fcd5965294db07ab92 _secondarycolorname_30_fc416cd64b531a33229a6f92a9637ec7 _primarycolor_33_0df58f7b4fd76e3730280f87cbd7972c _secondarycolor_35_80445e244cf10fd2ea84e6ba42cd4470 _icontexture_38_a00106cb465c5e27125b43b8120d23d6 _iconcolor_41_6e822cf54751d12c350ce0b5712e94cb _showiconflashing_48_6c44a7664cd2b711a38e888f2d4bec23 _showtextalert_49_bb469ca94efc177da4282789c3a6edfd) (Utilities|Struct|BreakStructBRGameUIGamePhaseVisuals (Utilities|Map|Find (LocalVariables|GetGamePhaseStepToVisualsMap) GamePhaseStep)))
  (bind _gamephasesteptovisualsmap (LocalVariables|GetGamePhaseStepToVisualsMap))
  (bind (_value _returnvalue) (Utilities|Map|Find _gamephasesteptovisualsmap GamePhaseStep))
  (bind _worldcontext (Variables|Get__WorldContext))
  (if _returnvalue
    (bind _outputcolor (CallFunction|GetOverrideColor _primarycolorname_28_e8470e2a467ff6fcd5965294db07ab92 _primarycolor_33_0df58f7b4fd76e3730280f87cbd7972c _worldcontext))
    (LocalVariables|SetPrimary%20Color _outputcolor)
    (bind _outputcolor_2 (CallFunction|GetOverrideColor _secondarycolorname_30_fc416cd64b531a33229a6f92a9637ec7 _secondarycolor_35_80445e244cf10fd2ea84e6ba42cd4470 _worldcontext))
    (LocalVariables|SetSecondary%20Color _outputcolor_2)
    (return (Utilities|Struct|MakeStructBRGameUIGamePhaseVisuals _messagetext_4_ca9786984f68487451e3e6bf33553019 _primarycolorname_28_e8470e2a467ff6fcd5965294db07ab92 _secondarycolorname_30_fc416cd64b531a33229a6f92a9637ec7 (LocalVariables|GetPrimary%20Color) (LocalVariables|GetSecondary%20Color) _icontexture_38_a00106cb465c5e27125b43b8120d23d6 _iconcolor_41_6e822cf54751d12c350ce0b5712e94cb _showiconflashing_48_6c44a7664cd2b711a38e888f2d4bec23 _showtextalert_49_bb469ca94efc177da4282789c3a6edfd))
    (else
      ;[1 0.5234 0.5052 0.5] Failed to find the visual map
      (Fortnite|FortShippingLog (Variables|Get__WorldContext) (Utilities|String|Append "GetGamePhaseVisualsFromStep, unable to find visuals for step: [" (Utilities|String|EnumtoString GamePhaseStep) "].") true)
      (return (Utilities|Struct|MakeStructBRGameUIGamePhaseVisuals "")))))

(fn GetOverrideColor
  (params (ParameterColorName :type Name) (OriginalColor :type Linear%20Color%20Structure))
  (if (and (!= ParameterColorName "") (!= ParameterColorName ""))
    (bind _returnvalue (Rendering|Material|GetVectorParameterValue (Variables|Get__WorldContext) "/BattleRoyaleGameUI/Shared/MPC_BRGameUI.MPC_BRGameUI" ParameterColorName :!class /Script/Engine.KismetMaterialLibrary))
    (return _returnvalue)
    (else
      (return OriginalColor))))
